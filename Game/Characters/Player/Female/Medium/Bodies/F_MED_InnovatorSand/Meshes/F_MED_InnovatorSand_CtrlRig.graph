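# Control Rig for the female medium InnovatorSand player body that applies pose-driven corrective offsets to extra thigh, back and upper-arm bones.
# Spherical pose readers on thighs and left upper arm weight those offsets.
# The left thigh reading, stored in a variable, is multiplied by smoothed back_volume or crouching_Alpha inputs; right thigh and left arm use raw readings.
# Right-arm and second left-arm readers, plus the jumping blend, drive nothing.
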ControlRigBlueprint /Game/Characters/Player/Female/Medium/Bodies/F_MED_InnovatorSand/Meshes/F_MED_InnovatorSand_CtrlRig
graphs [
 ContainedGraph {nodes [
   n1 = RigVMFunctionEntryNode()
   SequenceAggregate = &RigVMFunction_Sequence(ExecuteContext=@n1.ExecuteContext)
   SequenceAggregate_1 = &RigVMFunction_Sequence(ExecuteContext=@SequenceAggregate.B)
   SequenceAggregate_2 = &RigVMFunction_Sequence(ExecuteContext=@SequenceAggregate_1.B)
   SequenceAggregate_3 = &RigVMFunction_Sequence(ExecuteContext=@SequenceAggregate_2.B)
   SequenceAggregate_4 = &RigVMFunction_Sequence(ExecuteContext=@SequenceAggregate_3.B)
   SequenceAggregate_5 = &RigVMFunction_Sequence(ExecuteContext=@SequenceAggregate_4.B)
   SequenceAggregate_6 = &RigVMFunction_Sequence(ExecuteContext=@SequenceAggregate_5.B)
   SequenceAggregate_7 = &RigVMFunction_Sequence(ExecuteContext=@SequenceAggregate_6.B)
   SequenceAggregate_8 = &RigVMFunction_Sequence(ExecuteContext=@SequenceAggregate_7.B)
   SequenceAggregate_9 = &RigVMFunction_Sequence(ExecuteContext=@SequenceAggregate_8.B)
   RigVMFunctionReturnNode(
    A=@SequenceAggregate.A
    B=@SequenceAggregate_1.A
    C=@SequenceAggregate_2.A
    D=@SequenceAggregate_3.A
    E=@SequenceAggregate_4.A
    F=@SequenceAggregate_5.A
    G=@SequenceAggregate_6.A
    H=@SequenceAggregate_7.A
    I=@SequenceAggregate_8.A
    J=@SequenceAggregate_9.A
    K=@SequenceAggregate_9.B
   )
  ]}
 RigVMFunctionLibrary {links [], nodes []}
 RigVMModel {nodes [
   RigUnit_BeginExecution = &RigUnit_BeginExecution()
   SequenceAggregate_1_1_1 = RigVMAggregateNode(ExecuteContext=@RigUnit_BeginExecution.ExecutePin)
   n1 = &RigUnit_SphericalPoseReader(
    ActiveRegionScaleFactors=(PositiveWidth=3.5,NegativeWidth=0.5,PositiveHeight=3,NegativeHeight=3)
    ActiveRegionSize=0.1
    Debug=(bDrawDebug=false,DebugScale=25,DebugSegments=20,DebugThickness=0.25)
    DriverAxis=(X=-1,Y=0,Z=0)
    DriverItem=(Type=Bone,Name="thigh_l")
    FalloffRegionScaleFactors=(PositiveWidth=1,NegativeWidth=1,PositiveHeight=1,NegativeHeight=1)
    FalloffSize=1
    OptionalParentItem=(Type=Bone)
    RotationOffset=(X=90,Y=0,Z=0)
    ExecutePin=@SequenceAggregate_1_1_1.A
   )
   > RigVMVariableNode(Value=@n1.OutputParam) {variable=thigh_l_SP}
   n2 = &RigVMFunction_AlphaInterp(InterpSpeedDecreasing=3, InterpSpeedIncreasing=5, Scale=1, bClampResult=False, Value=RigVMVariableNode() {variable=back_volume}.Value)
   > &RigUnit_ModifyTransforms(
    ItemToModify=((Item=(Type=Bone,Name="extra_back_thigh_l"),Transform=(Rotation=(X=0,Y=0,Z=0,W=1),Translation=(X=53,Y=-16.5,Z=0),Scale3D=(X=1,Y=1,Z=1))))
    Weight=&RigVMFunction_MathDoubleMul(A=@n2.Result, B=RigVMVariableNode() {variable=thigh_l_SP}.Value).Result
   )
   n3 = &RigVMFunction_AlphaInterp(InterpSpeedDecreasing=3, InterpSpeedIncreasing=5, Scale=1, bClampResult=False, Value=RigVMVariableNode() {variable=crouching_Alpha}.Value)
   > RigUnit_SetTranslation = /Script/ControlRig.RigUnit_SetTranslation(
    Item=(Type=Bone,Name="extra_front_thigh_l")
    Space=LocalSpace
    Value=(X=-14.25,Y=42,Z=0.25)
    bInitial=False
    bPropagateToChildren=True
    Weight=&RigVMFunction_MathDoubleMul(A=@n3.Result, B=RigVMVariableNode() {variable=thigh_l_SP}.Value).Result
   ) {is=struct RigVMUnitNode, template="Set Transform::Execute(in Item,in Space,in bInitial,in Value,in Weight,in bPropagateToChildren)"}
   n4 = &RigVMFunction_AlphaInterp(InterpSpeedDecreasing=3, InterpSpeedIncreasing=5, Scale=1, bClampResult=False, Value=RigVMVariableNode() {variable=back_volume}.Value)
   > &RigUnit_SetTranslation(
    Item=(Type=Bone,Name="extra_back")
    Space=LocalSpace
    Value=(X=-20,Y=-7.6,Z=-2.6)
    bInitial=False
    bPropagateToChildren=True
    Weight=&RigVMFunction_MathDoubleMul(A=@n4.Result, B=RigVMVariableNode() {variable=thigh_l_SP}.Value).Result
   ) {template="Set Transform::Execute(in Item,in Space,in bInitial,in Value,in Weight,in bPropagateToChildren)"}
   n5 = &RigUnit_SphericalPoseReader(
    ActiveRegionScaleFactors=(PositiveWidth=3.5,NegativeWidth=0.5,PositiveHeight=3,NegativeHeight=3)
    ActiveRegionSize=0.1
    Debug=(bDrawDebug=false,DebugScale=25,DebugSegments=20,DebugThickness=0.25)
    DriverAxis=(X=1,Y=0,Z=0)
    DriverItem=(Type=Bone,Name="thigh_r")
    FalloffRegionScaleFactors=(PositiveWidth=1,NegativeWidth=1,PositiveHeight=1,NegativeHeight=1)
    FalloffSize=1
    FlipHeightScaling=false
    FlipWidthScaling=true
    OptionalParentItem=(Type=Bone)
    RotationOffset=(X=-90,Y=0,Z=0)
    ExecutePin=@SequenceAggregate_1_1_1.B
   )
   > &RigUnit_ModifyTransforms(
    ItemToModify=((Item=(Type=Bone,Name="extra_back_thigh_r"),Transform=(Rotation=(X=0,Y=0,Z=0,W=1),Translation=(X=0,Y=22,Z=-20),Scale3D=(X=1,Y=1,Z=1))))
    Weight=@n5.OutputParam
   )
   > &RigUnit_ModifyTransforms(
    ItemToModify=((Item=(Type=Bone,Name="extra_front_thigh_r"),Transform=(Rotation=(X=0,Y=0,Z=0,W=1),Translation=(X=0,Y=-8,Z=0),Scale3D=(X=1,Y=1,Z=1))))
    Weight=@n5.OutputParam
   )
   n6 = &RigUnit_SphericalPoseReader(
    ActiveRegionScaleFactors=(PositiveWidth=2.5,NegativeWidth=3,PositiveHeight=3,NegativeHeight=3)
    ActiveRegionSize=0.1
    Debug=(bDrawDebug=false,bDraw2D=false,bDrawLocalAxes=false,DebugScale=25,DebugSegments=20,DebugThickness=0.25)
    DriverAxis=(X=1,Y=0,Z=0)
    DriverItem=(Type=Bone,Name="upperarm_l")
    FalloffRegionScaleFactors=(PositiveWidth=1,NegativeWidth=1,PositiveHeight=1,NegativeHeight=1)
    FalloffSize=0.5
    OptionalParentItem=(Type=Bone)
    RotationOffset=(X=0,Y=0,Z=0)
    ExecutePin=@SequenceAggregate_1_1_1.C
   )
   > &RigUnit_SetTranslation(
    Item=(Type=Bone,Name="extra_upperarm_l")
    Space=LocalSpace
    Value=(X=0,Y=0,Z=-5)
    bInitial=False
    bPropagateToChildren=True
    Weight=@n6.OutputParam
   ) {template="Set Transform::Execute(in Item,in Space,in bInitial,in Value,in Weight,in bPropagateToChildren)"}
   &RigUnit_SphericalPoseReader(
    ActiveRegionScaleFactors=(PositiveWidth=2.5,NegativeWidth=3,PositiveHeight=3,NegativeHeight=3)
    ActiveRegionSize=0.1
    Debug=(bDrawDebug=false,DebugScale=25,DebugSegments=20,DebugThickness=0.25)
    DriverAxis=(X=-1,Y=0,Z=0)
    DriverItem=(Type=Bone,Name="upperarm_r")
    FalloffRegionScaleFactors=(PositiveWidth=1,NegativeWidth=1,PositiveHeight=1,NegativeHeight=1)
    FalloffSize=0.5
    FlipHeightScaling=true
    OptionalParentItem=(Type=Bone)
    RotationOffset=(X=180,Y=0,Z=180)
    ExecutePin=@SequenceAggregate_1_1_1.D
   )
   &RigUnit_SphericalPoseReader(
    ActiveRegionScaleFactors=(PositiveWidth=0.5,NegativeWidth=3,PositiveHeight=1.5,NegativeHeight=1.5)
    ActiveRegionSize=0.1
    Debug=(bDrawDebug=false,DebugScale=25,DebugSegments=20,DebugThickness=0.25)
    DriverAxis=(X=1,Y=0,Z=0)
    DriverItem=(Type=Bone,Name="upperarm_l")
    FalloffRegionScaleFactors=(PositiveWidth=1,NegativeWidth=1,PositiveHeight=1,NegativeHeight=1)
    FalloffSize=0.4
    OptionalParentItem=(Type=Bone)
    RotationOffset=(X=90,Y=0,Z=0)
    ExecutePin=@SequenceAggregate_1_1_1.E
   )
   &RigUnit_ModifyTransforms(
    ExecutePin=()
    ItemToModify=((Item=(Type=Bone,Name="extra_front_thigh_l"),Transform=(Rotation=(X=0,Y=0,Z=0,W=1),Translation=(X=0,Y=0,Z=0),Scale3D=(X=1,Y=1,Z=1))))
    Weight=1
    ItemToModify.0.Transform.Rotation=&RigVMFunction_MathQuaternionFromEuler(Euler=(X=0,Y=0,Z=0)).Result
   )
   n7 = &RigVMFunction_AlphaInterp(InterpSpeedDecreasing=3, InterpSpeedIncreasing=5, Scale=1, bClampResult=False, Value=RigVMVariableNode() {variable=jumping}.Value)
   &RigVMFunction_MathDoubleMul(A=@n7.Result, B=@n5.OutputParam)
  ]}
]
hierarchy [
 root {loc=<0 0 0>, rot=<0 0 0 1>}
 pelvis {parent=root, loc=<0 -0.5301 98.0571>, rot=<0.0127 0.707 -0.0127 -0.707>}
 spine_01 {parent=pelvis, loc=<6.1208 -0.9328 0>, rot=<0 0 0.034 -0.9994>}
 spine_02 {parent=spine_01, loc=<6.5996 -0.2095 0>, rot=<0 0 0.0238 -0.9997>}
 spine_03 {parent=spine_02, loc=<6.4914 -0.1237 0>, rot=<0 0 0.0149 -0.9999>}
 spine_04 {parent=spine_03, loc=<6.9598 0.0794 0>, rot=<0 0 -0.0132 -0.9999>}
 spine_05 {parent=spine_04, loc=<6.9987 0 0>, rot=<0 0 -0.0015 -1>}
 attach_cape {parent=spine_05, loc=<9.3216 -8.4501 0>, rot=<0.0616 -0.7044 -0.0616 -0.7044>}
 neck_01 {parent=spine_05, loc=<11.3138 0.1529 0>, rot=<0 0 -0.2003 -0.9797>}
 neck_02 {parent=neck_01, loc=<4.5627 0 0>, rot=<0 0 0.0504 -0.9987>}
 head {parent=neck_02, loc=<5.3914 0 0>, rot=<0 0 0.115 -0.9934>}
 attach_backpack {parent=spine_05, loc=<0 -8.35 0>, rot=<-0.0093 -0.707 0.0093 -0.707>}
 clavicle_l {parent=spine_05, loc=<5.4831 -0.1053 -3.2946>, rot=<0.0465 -0.7289 -0.009 -0.6829>}
 upperarm_l {parent=clavicle_l, loc=<9.8658 -1.2136 -0.4105>, rot=<0.0164 -0.3245 -0.0269 -0.9454>}
 lowerarm_l {parent=upperarm_l, loc=<26.465 -0.8345 0.259>, rot=<0.0036 0.0161 -0.2298 -0.9731>}
 hand_l {parent=lowerarm_l, loc=<22.7355 -0.3203 -0.6105>, rot=<-0.6784 -0.0223 0.0735 -0.7307>}
 thumb_01_l {parent=hand_l, loc=<1.3118 -0.3191 -2.459>, rot=<0.6546 -0.4042 0.0664 -0.6354>}
 thumb_02_l {parent=thumb_01_l, loc=<5.9336 -0.0448 0.0401>, rot=<-0.0037 0.0213 0.298 -0.9543>}
 thumb_03_l {parent=thumb_02_l, loc=<2.2223 0.3682 -0.0907>, rot=<0.0118 -0.013 0.0066 -0.9998>}
 index_metacarpal_l {parent=hand_l, loc=<1.6535 0.0183 -2.96>, rot=<0.0158 -0.0124 -0.0173 -0.9996>}
 index_01_l {parent=index_metacarpal_l, loc=<8.3008 0.0357 -0.0524>, rot=<-0.0044 0.0485 0.0924 -0.9945>}
 index_02_l {parent=index_01_l, loc=<3.3587 -0.3004 0.0452>, rot=<-0.002 0.0027 0.3225 -0.9466>}
 index_03_l {parent=index_02_l, loc=<2.5578 -0.123 -0.0252>, rot=<-0.002 0.0067 0.2558 -0.9667>}
 middle_metacarpal_l {parent=hand_l, loc=<2.1856 -0.1725 -0.5984>, rot=<0.006 0.0168 -0.0329 -0.9993>}
 middle_01_l {parent=middle_metacarpal_l, loc=<7.3275 -0.2447 0.0129>, rot=<-0.0208 0.0042 0.0858 -0.9961>}
 middle_02_l {parent=middle_01_l, loc=<4.0123 0.1401 0.1175>, rot=<-0.0218 0.0258 0.3512 -0.9357>}
 middle_03_l {parent=middle_02_l, loc=<2.6077 0.0534 0.1373>, rot=<0.0082 0.0188 0.1708 -0.9851>}
 ring_metacarpal_l {parent=hand_l, loc=<2.2074 -0.3625 1.3638>, rot=<-0.0284 0.0722 -0.0458 -0.9959>}
 ring_01_l {parent=ring_metacarpal_l, loc=<7.0295 -0.6197 -0.1446>, rot=<-0.0489 -0.0252 0.1892 -0.9804>}
 ring_02_l {parent=ring_01_l, loc=<3.6174 0.4647 -0.1027>, rot=<0 0 0.2806 -0.9598>}
 ring_03_l {parent=ring_02_l, loc=<2.4152 0.114 0.0199>, rot=<0 0 0.1189 -0.9929>}
 pinky_metacarpal_l {parent=hand_l, loc=<1.7616 -0.9753 3.3814>, rot=<-0.0397 0.0814 -0.069 -0.9935>}
 pinky_01_l {parent=pinky_metacarpal_l, loc=<7.1709 -0.4831 -0.1094>, rot=<-0.0654 -0.0035 0.2242 -0.9723>}
 pinky_02_l {parent=pinky_01_l, loc=<2.8274 0.0893 -0.1043>, rot=<0.0037 0.0011 0.2854 -0.9584>}
 pinky_03_l {parent=pinky_02_l, loc=<2.2046 0.0307 -0.0078>, rot=<-0.0002 0.0098 0.1445 -0.9895>}
 lowerarm_twist_02_l {parent=lowerarm_l, loc=<10.6707 0 0>, rot=<0 0 0 1>}
 lowerarm_twist_01_l {parent=lowerarm_l, loc=<18.316 0 0>, rot=<0 0 0 1>}
 upperarm_twist_01_l {parent=upperarm_l, loc=<7.2135 -0.1134 0>, rot=<0 0 0 1>}
 upperarm_twist_02_l {parent=upperarm_l, loc=<19.7998 -0.1134 0>, rot=<0 0 0 1>}
 extra_upperarm_l {parent=upperarm_l, loc=<0 0 0>, rot=<-0.0004 0.0034 0.0158 -0.9999>}
 extra_clavicle_l {parent=clavicle_l, loc=<0 0 0>, rot=<0 0 0 1>}
 clavicle_r {parent=spine_05, loc=<5.4831 -0.1053 3.2946>, rot=<0.7289 0.0465 -0.6829 0.009>}
 upperarm_r {parent=clavicle_r, loc=<-9.8658 1.2136 0.4105>, rot=<0.0164 -0.3245 -0.0269 -0.9454>}
 lowerarm_r {parent=upperarm_r, loc=<-26.465 0.8345 -0.259>, rot=<0.0036 0.0161 -0.2298 -0.9731>}
 hand_r {parent=lowerarm_r, loc=<-22.7355 0.3203 0.6106>, rot=<-0.6784 -0.0223 0.0735 -0.7307>}
 thumb_01_r {parent=hand_r, loc=<-1.3118 0.3191 2.459>, rot=<0.6546 -0.4042 0.0664 -0.6354>}
 thumb_02_r {parent=thumb_01_r, loc=<-5.9336 0.0448 -0.0401>, rot=<-0.0037 0.0213 0.298 -0.9543>}
 thumb_03_r {parent=thumb_02_r, loc=<-2.2223 -0.3682 0.0907>, rot=<0.0118 -0.013 0.0066 -0.9998>}
 index_metacarpal_r {parent=hand_r, loc=<-1.6535 -0.0183 2.96>, rot=<0.0158 -0.0124 -0.0173 -0.9996>}
 index_01_r {parent=index_metacarpal_r, loc=<-8.3008 -0.0357 0.0524>, rot=<-0.0044 0.0485 0.0924 -0.9945>}
 index_02_r {parent=index_01_r, loc=<-3.3587 0.3004 -0.0452>, rot=<-0.002 0.0027 0.3225 -0.9466>}
 index_03_r {parent=index_02_r, loc=<-2.5578 0.123 0.0252>, rot=<-0.002 0.0067 0.2558 -0.9667>}
 middle_metacarpal_r {parent=hand_r, loc=<-2.1856 0.1725 0.5984>, rot=<0.006 0.0168 -0.0329 -0.9993>}
 middle_01_r {parent=middle_metacarpal_r, loc=<-7.3275 0.2447 -0.0129>, rot=<-0.0208 0.0042 0.0858 -0.9961>}
 middle_02_r {parent=middle_01_r, loc=<-4.0123 -0.1401 -0.1175>, rot=<-0.0218 0.0258 0.3512 -0.9357>}
 middle_03_r {parent=middle_02_r, loc=<-2.6077 -0.0534 -0.1373>, rot=<0.0082 0.0188 0.1708 -0.9851>}
 ring_metacarpal_r {parent=hand_r, loc=<-2.2074 0.3625 -1.3638>, rot=<-0.0284 0.0722 -0.0458 -0.9959>}
 ring_01_r {parent=ring_metacarpal_r, loc=<-7.0295 0.6197 0.1446>, rot=<-0.0489 -0.0252 0.1892 -0.9804>}
 ring_02_r {parent=ring_01_r, loc=<-3.6174 -0.4647 0.1027>, rot=<0 0 0.2806 -0.9598>}
 ring_03_r {parent=ring_02_r, loc=<-2.4152 -0.114 -0.0199>, rot=<0 0 0.1189 -0.9929>}
 pinky_metacarpal_r {parent=hand_r, loc=<-1.7616 0.9753 -3.3814>, rot=<-0.0397 0.0814 -0.069 -0.9935>}
 pinky_01_r {parent=pinky_metacarpal_r, loc=<-7.1709 0.4831 0.1094>, rot=<-0.0654 -0.0035 0.2242 -0.9723>}
 pinky_02_r {parent=pinky_01_r, loc=<-2.8274 -0.0893 0.1043>, rot=<0.0037 0.0011 0.2854 -0.9584>}
 pinky_03_r {parent=pinky_02_r, loc=<-2.2046 -0.0307 0.0078>, rot=<-0.0002 0.0098 0.1445 -0.9895>}
 lowerarm_twist_02_r {parent=lowerarm_r, loc=<-10.6707 0 0>, rot=<0 0 0 -1>}
 lowerarm_twist_01_r {parent=lowerarm_r, loc=<-18.316 0 0>, rot=<0 0 0 -1>}
 upperarm_twist_01_r {parent=upperarm_r, loc=<-7.2135 0.1134 0>, rot=<0 0 0 -1>}
 upperarm_twist_02_r {parent=upperarm_r, loc=<-19.7998 0.1134 0>, rot=<0 0 0 -1>}
 extra_upperarm_r {parent=upperarm_r, loc=<-0.0001 0 0>, rot=<-0.0004 0.0034 0.0158 -0.9999>}
 dyn_M_rope_joint {parent=spine_05, loc=<3.0388 7.5761 -7.4661>, rot=<0.8477 0.5281 -0.0444 -0.0229>}
 dyn_UPPER_01_rope_joint {parent=dyn_M_rope_joint, loc=<1.1218 4.192 0.1871>, rot=<-0.628 -0.7535 -0.1948 0.0018>}
 dyn_UPPER_02_rope_joint {parent=dyn_UPPER_01_rope_joint, loc=<4.3933 0 0>, rot=<0.0381 -0.0371 0.1451 -0.988>}
 dyn_LOWER_01_rope_joint {parent=dyn_M_rope_joint, loc=<1.4188 -2.6861 0.5453>, rot=<-0.0694 0.1553 0.7223 -0.6703>}
 dyn_LOWER_02_rope_joint {parent=dyn_LOWER_01_rope_joint, loc=<3.971 0 0>, rot=<0 0 0 1>}
 extra_UPPER_plaque_joint {parent=spine_05, loc=<1.3342 8.4341 7.3801>, rot=<0.0284 -0.7065 -0.0284 -0.7065>}
 extra_M_plaque_joint {parent=spine_02, loc=<2.7707 9.7947 0>, rot=<0 0 -0.0578 -0.9983>}
 dyn_L_strap_BASE_joint {parent=spine_02, loc=<0.0312 7.3254 -6.872>, rot=<-0.0512 -0.9713 0.2275 -0.0476>}
 dyn_L_strap_01_joint {parent=dyn_L_strap_BASE_joint, loc=<0 0 0>, rot=<0 0 0 1>}
 dyn_L_strap_02_joint {parent=dyn_L_strap_01_joint, loc=<3.3736 0 0>, rot=<0 0 -0.0021 -1>}
 dyn_L_strap_03_joint {parent=dyn_L_strap_02_joint, loc=<5.5908 0 0>, rot=<0 0 0 1>}
 dyn_carabiner_01_joint {parent=dyn_L_strap_03_joint, loc=<3.2075 0.1967 -1.4835>, rot=<0.1822 -0.1994 -0.0547 -0.9613>}
 dyn_carabiner_02_joint {parent=dyn_L_strap_03_joint, loc=<3.5788 0.5384 -0.5786>, rot=<0.2163 -0.0949 -0.0204 -0.9715>}
 dyn_carabiner_03_joint {parent=dyn_L_strap_03_joint, loc=<3.8435 0.4963 0.3462>, rot=<0.2424 0.0332 -0.0389 -0.9688>}
 dyn_carabiner_04_joint {parent=dyn_L_strap_03_joint, loc=<3.0774 0.8281 1.4124>, rot=<0.2509 0.174 0.0007 -0.9523>}
 dyn_R_strap_BASE_joint {parent=spine_02, loc=<-1.071 7.4337 7.3986>, rot=<-0.0923 -0.9623 -0.2441 0.0767>}
 dyn_R_strap_01_joint {parent=dyn_R_strap_BASE_joint, loc=<0 0 0>, rot=<0 0 0 1>}
 extra_LOWER_plaque_joint {parent=spine_01, loc=<-0.5652 8.3888 0>, rot=<0 0 -0.034 -0.9994>}
 dyn_waist_pouch_joint {parent=spine_01, loc=<5.0446 -5.2513 -5.8731>, rot=<-0.0433 -0.1212 0.9639 -0.2331>}
 thigh_l {parent=pelvis, loc=<-1.0855 -0.2109 -8.7053>, rot=<0.0009 0.0355 0.0155 -0.9993>}
 calf_l {parent=thigh_l, loc=<-42.5284 -0.0019 -0.7297>, rot=<0.0006 -0.0161 -0.0391 -0.9991>}
 foot_l {parent=calf_l, loc=<-46.7401 -2.5014 -0.592>, rot=<0.0427 -0.0238 0.0353 -0.9982>}
 ball_l {parent=foot_l, loc=<-3.5658 10.8066 -0.1544>, rot=<0.0001 -0.0001 -0.7063 -0.7079>}
 calf_twist_02_l {parent=calf_l, loc=<-19.1546 -1.7943 -0.2638>, rot=<0.0237 0.0095 -0.0035 -0.9997>}
 calf_twist_01_l {parent=calf_l, loc=<-35.8483 -1.9197 -0.5794>, rot=<0.0237 0.0095 -0.0035 -0.9997>}
 thigh_twist_01_l {parent=thigh_l, loc=<-27.952 0 0>, rot=<0.0013 0 0 -1>}
 extra_back_thigh_l {parent=thigh_l, loc=<-0.2028 -4.628 0.0007>, rot=<0 -0.001 0 -1>}
 extra_front_thigh_l {parent=thigh_l, loc=<-3.3417 3.2879 0.2216>, rot=<0 -0.001 0 -1>}
 dyn_pouch_front_leg_joint {parent=thigh_l, loc=<-15.2523 7.4928 -1.0895>, rot=<-0.0024 -1 -0.0002 -0.005>}
 dyn_BASE_zipper_front_leg_joint {parent=dyn_pouch_front_leg_joint, loc=<0.2548 2.5798 2.0434>, rot=<0.6537 -0.0306 -0.0265 -0.7556>}
 dyn_zipper_front_leg_joint {parent=dyn_BASE_zipper_front_leg_joint, loc=<0 0 0>, rot=<0 0 0 1>}
 dyn_pouch_side_leg_joint {parent=thigh_l, loc=<-15.2186 0.7666 -7.7291>, rot=<0.0105 -0.7052 0.7088 0.014>}
 dyn_BASE_zipper_side_leg_joint {parent=dyn_pouch_side_leg_joint, loc=<0.3042 2.3606 2.1229>, rot=<0.631 -0.0113 0.0142 -0.7756>}
 dyn_zipper_side_leg_joint {parent=dyn_BASE_zipper_side_leg_joint, loc=<0 0 0>, rot=<0 0 0 1>}
 dyn_pouch_back_leg_joint {parent=thigh_l, loc=<-15.408 -7.3053 -2.2552>, rot=<0.0169 -0.1276 0.9917 0.0047>}
 dyn_BASE_zipper_back_leg_joint {parent=dyn_pouch_back_leg_joint, loc=<0.3767 2.5431 1.9864>, rot=<0.6802 0.0098 0.0152 -0.7328>}
 dyn_zipper_back_leg_joint {parent=dyn_BASE_zipper_back_leg_joint, loc=<0 0 0>, rot=<0 0 0 1>}
 thigh_r {parent=pelvis, loc=<-1.0855 -0.2109 8.7053>, rot=<0.0355 -0.0009 0.9993 0.0155>}
 calf_r {parent=thigh_r, loc=<42.5284 0.0019 0.7297>, rot=<0.0006 -0.0161 -0.0391 -0.9991>}
 foot_r {parent=calf_r, loc=<46.7401 2.5014 0.592>, rot=<0.0427 -0.0238 0.0353 -0.9982>}
 ball_r {parent=foot_r, loc=<3.5658 -10.8066 0.1544>, rot=<0.0001 -0.0001 -0.7063 -0.7079>}
 calf_twist_02_r {parent=calf_r, loc=<19.1546 1.7943 0.2638>, rot=<0.0237 0.0095 -0.0035 -0.9997>}
 calf_twist_01_r {parent=calf_r, loc=<35.8483 1.9197 0.5794>, rot=<0.0237 0.0095 -0.0035 -0.9997>}
 thigh_twist_01_r {parent=thigh_r, loc=<27.952 0 0>, rot=<0.0013 0 0 -1>}
 extra_back_thigh_r {parent=thigh_r, loc=<0.2028 4.628 -0.0007>, rot=<0 -0.001 0 -1>}
 extra_front_thigh_r {parent=thigh_r, loc=<3.3413 -3.2883 -0.2214>, rot=<0 -0.001 0 -1>}
 extra_back {parent=pelvis, loc=<0 0 0>, rot=<0 0 0 1>}
]
variables [
 back_volume {type=double, default=0}
 crouching_Alpha {type=double, default=0}
 jumping {type=double, default=0}
 thigh_l_SP {type=double, default=0}
]
