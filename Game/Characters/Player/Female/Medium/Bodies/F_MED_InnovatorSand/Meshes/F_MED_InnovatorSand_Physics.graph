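# Physics asset for the F_MED_InnovatorSand female player body, set up to simulate its dangling straps, carabiners and waist pouch.
# Pelvis, spine, thigh and lower plaque shapes form the body volumes the accessories hang around.
# Both strap bases and the pouch constrain to spine_01; the left strap is a three-link chain ending in four carabiners, the right a single link.
# Tight swing and twist limits restrain most accessory joints, while base joints rotate freely.
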PhysicsAsset /Game/Characters/Player/Female/Medium/Bodies/F_MED_InnovatorSand/Meshes/F_MED_InnovatorSand_Physics
bodies {
 dyn_L_strap_01_joint [{shapeType=Box, extentX=2.9896, extentY=2.0696, extentZ=4.3261, shapeName=dyn_L_strap_01_joint_box, center=<1.7784 0.608 -0.0985>}]
 dyn_L_strap_02_joint [{shapeType=Box, extentX=5.0264, extentY=1.5002, extentZ=3.8264, shapeName=dyn_L_strap_02_joint_box, center=<2.654 0 0.0912>}]
 dyn_L_strap_03_joint [{
   shapeType=Capsule
   length=2.8638
   radius=2.8636
   shapeName=dyn_L_strap_03_joint_capsule
   center=<1.2807 2.0038 0.0725>
   rotation=<rot 0 104.0469 -3.5383>
  }]
 dyn_L_strap_BASE_joint [{shapeType=Capsule, length=0.505, radius=0.505, shapeName=dyn_L_strap_BASE_joint_capsule, rotation=<rot 0 -90 0>}]
 dyn_R_strap_01_joint [{
   shapeType=Capsule
   length=3.3165
   radius=2.4819
   shapeName=dyn_R_strap_01_joint_capsule
   center=<4.011 0.8014 0.1938>
   rotation=<rot -87.2342 -0.0002 -2.3641>
  }]
 dyn_R_strap_BASE_joint [{shapeType=Capsule, length=0.505, radius=0.505, shapeName=dyn_R_strap_BASE_joint_capsule, rotation=<rot 0 -90 0>}]
 dyn_carabiner_01_joint [{shapeType=Box, extentX=4.0219, extentY=4.7651, extentZ=1.3471, shapeName=dyn_carabiner_01_joint_box, center=<4.8732 0 0>}]
 dyn_carabiner_02_joint [{
   shapeType=Box
   extentX=3.8313
   extentY=4.3017
   extentZ=1.2875
   shapeName=dyn_carabiner_02_joint_box
   center=<4.6688 0.0058 0.0235>
   rotation=<rot 0 0.0002 -6.7699>
  }]
 dyn_carabiner_03_joint [{
   shapeType=Box
   extentX=3.8313
   extentY=4.6869
   extentZ=1.4148
   shapeName=dyn_carabiner_03_joint_box
   center=<4.6564 -0.0122 -0.0241>
   rotation=<rot -0.1875 -0.7319 -16.8967>
  }]
 dyn_carabiner_04_joint [{
   shapeType=Box
   extentX=3.9327
   extentY=4.5651
   extentZ=1.3766
   shapeName=dyn_carabiner_04_joint_box
   center=<4.7944 0 0>
   rotation=<rot 0 0.0001 -19.6173>
  }]
 dyn_waist_pouch_joint [{
   shapeType=Capsule
   length=3.2935
   radius=5.307
   shapeName=dyn_waist_pouch_joint_capsule
   center=<4.9618 5.1953 -0.1222>
   rotation=<rot 9.9407 -93.8837 -82.7554>
  }]
 extra_LOWER_plaque_joint [{
   shapeType=Box
   extentX=12.7453
   extentY=1.4567
   extentZ=5.946
   shapeName=extra_LOWER_plaque_joint_box
   center=<3.2925 0.1743 0.3936>
   rotation=<rot 3.4571 -0.0001 0>
  }]
 pelvis [{shapeType=Capsule, length=12.535, radius=9.8679, center=<-3.8186 -0.1159 0>}]
 spine_01 [
  {shapeType=Capsule, length=14.6374, radius=6.5334, shapeName=spine_01_capsule}
  {
   shapeType=Capsule
   length=24.6875
   radius=5.3125
   shapeName=spine_01_capsule
   center=<-4.5894 -5.5503 0.6685>
   rotation=<rot 0 -0.0062 -18.2218>
  }
 ]
 spine_02 [{shapeType=Capsule, length=8.9375, radius=6.0625, shapeName=spine_02_capsule, center=<-2.215 1.5448 0>}]
 thigh_l [{shapeType=Capsule, length=24.2261, radius=8.4152, center=<-16.6226 -0.3884 -0.2082>, rotation=<rot 1.3014 -91.0819 -91.7936>}]
 thigh_r [{shapeType=Capsule, length=21.6171, radius=9.0869, center=<11.4368 0.449 -0.3632>, rotation=<rot 0 -87.3303 -85.2348>}]
}
constraints [
 {bone1Name=thigh_l, bone2Name=pelvis, swing1LimitDegrees=45, swing2LimitDegrees=45, twistLimitDegrees=45}
 {bone1Name=thigh_r, bone2Name=pelvis, swing1LimitDegrees=45, swing2LimitDegrees=45, twistLimitDegrees=45}
 {
  bone1Name=dyn_R_strap_BASE_joint
  bone2Name=spine_01
  swing1LimitDegrees=45
  swing1Motion=Free
  swing2LimitDegrees=45
  swing2Motion=Free
  twistLimitDegrees=45
  twistMotion=Free
 }
 {
  bone1Name=dyn_R_strap_01_joint
  bone2Name=dyn_R_strap_BASE_joint
  swing1LimitDegrees=100
  swing2LimitDegrees=5
  swing2Motion=Locked
  twistLimitDegrees=3
 }
 {
  bone1Name=dyn_L_strap_BASE_joint
  bone2Name=spine_01
  swing1LimitDegrees=45
  swing1Motion=Free
  swing2LimitDegrees=45
  swing2Motion=Free
  twistLimitDegrees=45
  twistMotion=Free
 }
 {
  bone1Name=dyn_L_strap_01_joint
  bone2Name=dyn_L_strap_BASE_joint
  swing1LimitDegrees=45
  swing1Motion=Free
  swing2LimitDegrees=45
  swing2Motion=Free
  twistLimitDegrees=5
 }
 {bone1Name=dyn_L_strap_02_joint, bone2Name=dyn_L_strap_01_joint, swing1LimitDegrees=20, swing2LimitDegrees=5, twistLimitDegrees=3}
 {bone1Name=dyn_L_strap_03_joint, bone2Name=dyn_L_strap_02_joint, swing1LimitDegrees=20, swing2LimitDegrees=5, twistLimitDegrees=3}
 {bone1Name=dyn_carabiner_01_joint, bone2Name=dyn_L_strap_03_joint, swing1LimitDegrees=15, swing2LimitDegrees=10, twistLimitDegrees=3}
 {bone1Name=dyn_carabiner_02_joint, bone2Name=dyn_L_strap_03_joint, swing1LimitDegrees=15, swing2LimitDegrees=10, twistLimitDegrees=3}
 {bone1Name=dyn_carabiner_03_joint, bone2Name=dyn_L_strap_03_joint, swing1LimitDegrees=15, swing2LimitDegrees=10, twistLimitDegrees=3}
 {bone1Name=dyn_carabiner_04_joint, bone2Name=dyn_L_strap_03_joint, swing1LimitDegrees=15, swing2LimitDegrees=10, twistLimitDegrees=3}
 {
  bone1Name=spine_01
  bone2Name=pelvis
  swing1LimitDegrees=45
  swing1Motion=Free
  swing2LimitDegrees=45
  swing2Motion=Free
  twistLimitDegrees=45
  twistMotion=Free
 }
 {
  bone1Name=extra_LOWER_plaque_joint
  bone2Name=spine_01
  swing1LimitDegrees=45
  swing1Motion=Free
  swing2LimitDegrees=45
  swing2Motion=Free
  twistLimitDegrees=45
  twistMotion=Free
 }
 {bone1Name=dyn_waist_pouch_joint, bone2Name=spine_01, swing1LimitDegrees=45, swing2LimitDegrees=10, twistLimitDegrees=20}
]
