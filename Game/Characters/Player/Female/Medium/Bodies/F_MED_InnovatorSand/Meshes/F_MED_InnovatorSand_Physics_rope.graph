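# Physics asset for the rope on the F_MED_InnovatorSand player body, covering dynamic rope joints and the nearby upper body.
# Capsule bodies cover the upper spine, left clavicle and left upper arm, with the clavicle free to swing and twist against spine_05.
# A middle rope joint hangs off spine_05 and branches into two-link upper and lower chains.
# Chain constraints have tight, asymmetric angular limits, narrower on the upper chain than the lower.
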PhysicsAsset /Game/Characters/Player/Female/Medium/Bodies/F_MED_InnovatorSand/Meshes/F_MED_InnovatorSand_Physics_rope
bodies {
 clavicle_l [{
   shapeType=Capsule
   length=7.8489
   radius=5.0185
   shapeName=clavicle_l_capsule
   center=<5.2809 -0.6148 -0.3169>
   rotation=<rot -76.5179 179.9617 -179.96>
  }]
 dyn_LOWER_01_rope_joint [{
   shapeType=Capsule
   length=1.646
   radius=3.1239
   shapeName=dyn_LOWER_01_rope_joint_capsule
   center=<2.6626 0.4869 -0.0102>
   rotation=<rot 2.0343 88.878 -89.6421>
  }]
 dyn_LOWER_02_rope_joint [{
   shapeType=Capsule
   length=2.2876
   radius=2.4731
   shapeName=dyn_LOWER_02_rope_joint_capsule
   center=<1.6267 -0.0646 0.1864>
   rotation=<rot -12.3456 -90 -5.9303>
  }]
 dyn_M_rope_joint [{shapeType=Capsule, length=0.7452, radius=0.5346}]
 dyn_UPPER_01_rope_joint [{
   shapeType=Capsule
   length=2.5178
   radius=2.8108
   shapeName=dyn_UPPER_01_rope_joint_capsule
   center=<1.7377 -0.0582 0.1849>
   rotation=<rot -0.4479 -89.9545 84.0653>
  }]
 dyn_UPPER_02_rope_joint [{
   shapeType=Capsule
   length=3.4781
   radius=3.3615
   shapeName=dyn_UPPER_02_rope_joint_capsule
   center=<4.504 -0.2203 -0.0569>
   rotation=<rot -0.2289 -95.9995 -79.4793>
  }]
 spine_04 [{
   shapeType=Capsule
   length=11.4728
   radius=9.7381
   shapeName=spine_04_capsule
   center=<-1.0161 0.6873 -0.0175>
   rotation=<rot 0.8307 -90 0.3983>
  }]
 spine_05 [{shapeType=Capsule, length=4.6066, radius=9.1717}]
 upperarm_l [
  {shapeType=Capsule, length=16.9238, radius=4.0474, center=<13.039 -0.3062 -0.323>, rotation=<rot 0.6113 89.0205 -86.7225>}
  {
   shapeType=Capsule
   length=17.8208
   radius=3.388
   shapeName=upperarm_l_capsule
   center=<12.5444 -5.8653 -0.2324>
   rotation=<rot -88.0874 0.0016 -0.0016>
  }
 ]
}
constraints [
 {bone1Name=spine_05, bone2Name=spine_04, swing1LimitDegrees=45, swing2LimitDegrees=45, twistLimitDegrees=45}
 {bone1Name=upperarm_l, bone2Name=clavicle_l, swing1LimitDegrees=45, swing2LimitDegrees=45, twistLimitDegrees=45}
 {bone1Name=dyn_M_rope_joint, bone2Name=spine_05, swing1LimitDegrees=45, swing2LimitDegrees=45, twistLimitDegrees=45}
 {
  bone1Name=clavicle_l
  bone2Name=spine_05
  swing1LimitDegrees=45
  swing1Motion=Free
  swing2LimitDegrees=45
  swing2Motion=Free
  twistLimitDegrees=45
  twistMotion=Free
 }
 {bone1Name=dyn_LOWER_01_rope_joint, bone2Name=dyn_M_rope_joint, swing1LimitDegrees=25, swing2LimitDegrees=10, twistLimitDegrees=5}
 {bone1Name=dyn_LOWER_02_rope_joint, bone2Name=dyn_LOWER_01_rope_joint, swing1LimitDegrees=25, swing2LimitDegrees=10, twistLimitDegrees=5}
 {bone1Name=dyn_UPPER_01_rope_joint, bone2Name=dyn_M_rope_joint, swing1LimitDegrees=25, swing2LimitDegrees=5, twistLimitDegrees=3}
 {bone1Name=dyn_UPPER_02_rope_joint, bone2Name=dyn_UPPER_01_rope_joint, swing1LimitDegrees=25, swing2LimitDegrees=5, twistLimitDegrees=3}
]
