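# Animation Blueprint for the female medium InnovatorSand body, copying the attached parent's pose and adding secondary motion and corrective rigging.
# Bone constraints position plaque joints and, at LOD 1 and above despite the LOD_2_3 name, rotate left straps.
# Two RigidBody passes and three zipper AnimDynamics nodes then feed a Control Rig.
# The update event drives the rig with binary jump, crouch and back-volume values, the last requiring stationary, non-turning crouch.
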AnimBlueprint /Game/Characters/Player/Female/Medium/Bodies/F_MED_InnovatorSand/Meshes/F_MED_InnovatorSand_AnimBP
target_skeleton=&/F_MED_InnovatorSand_Skeleton
graphs [
 AnimGraph {nodes [
   n1 = "Constraint - dyn_L_strap_02_joint (RR)"!(
    ComponentPose=Local To Component(LocalPose=Copy Pose From Mesh() {bUseAttachedParent=true})
    bAlphaBoolEnabled:bool=Get LOD_2_3() -> (LOD_2_3:bool)
    ConstraintWeights_0=0
    ConstraintWeights_1:float=1
   ) -> (Pose) {
    alphaInputType=Bool
    alphaBoolBlend {blendOption=Linear}
    boneToModify {boneName=dyn_L_strap_02_joint}
    constraintSetup [{transformType=Rotation, targetBone {boneName=pelvis}}, {transformType=Rotation, targetBone {boneName=thigh_l}}]
    constraintWeights [1,1]
   }
   n2 = "Constraint - dyn_L_strap_03_joint (RR)"!(ComponentPose=@n1, bAlphaBoolEnabled:bool=Get LOD_2_3(), ConstraintWeights_0=0, ConstraintWeights_1:float=1) -> (Pose) {
    alphaInputType=Bool
    alphaBoolBlend {blendOption=Linear}
    boneToModify {boneName=dyn_L_strap_03_joint}
    constraintSetup [{transformType=Rotation, targetBone {boneName=pelvis}}, {transformType=Rotation, targetBone {boneName=thigh_l}}]
    constraintWeights [1,1]
   }
   n3 = "Constraint - extra_UPPER_plaque_joint (PPP)"!(ComponentPose=@n2, Alpha=1, ConstraintWeights_0=0.5966, ConstraintWeights_1:float=0.4024, ConstraintWeights_2:float=0.001) -> (Pose) {
    alphaBoolBlend {blendOption=Linear}
    boneToModify {boneName=extra_UPPER_plaque_joint}
    constraintSetup [{targetBone {boneName=spine_04}}, {targetBone {boneName=spine_05}}, {targetBone {boneName=clavicle_r}}]
    constraintWeights [1,1,1]
   }
   n4 = "Constraint - extra_M_plaque_joint (PPPPP)"!(
    ComponentPose=@n3
    Alpha=1
    ConstraintWeights_0=0.3463
    ConstraintWeights_1:float=0.2773
    ConstraintWeights_2:float=0.2022
    ConstraintWeights_3:float=0.1461
    ConstraintWeights_4:float=0.028
   ) -> (Pose) {
    alphaBoolBlend {blendOption=Linear}
    boneToModify {boneName=extra_M_plaque_joint}
    constraintSetup [
     {targetBone {boneName=spine_02}}
     {targetBone {boneName=spine_04}}
     {targetBone {boneName=spine_01}}
     {targetBone {boneName=spine_03}}
     {targetBone {boneName=pelvis}}
    ]
    constraintWeights [1,1,1,1,1]
   }
   n5 = "Constraint - extra_LOWER_plaque_joint (PP)"!(ComponentPose=@n4, Alpha=1, ConstraintWeights_0=0.783, ConstraintWeights_1:float=0.217) -> (Pose) {
    alphaBoolBlend {blendOption=Linear}
    boneToModify {boneName=extra_LOWER_plaque_joint}
    constraintSetup [{targetBone {boneName=pelvis}}, {targetBone {boneName=spine_01}}]
    constraintWeights [1,1]
   }
   n6 = RigidBody(ComponentPose=@n5) {
    bClampLinearTranslationLimitToRefPose=true
    bOverrideWorldGravity=true
    overridePhysicsAsset=&/F_MED_InnovatorSand_Physics_rope
    simulationSpace=BaseBoneSpace
    alphaBoolBlend {blendOption=Linear}
    alphaScaleBiasClamp {bInterpResult=true}
    baseBoneRef {boneName=spine_04}
    componentAppliedLinearAccClamp=<10 10 10>
    overrideWorldGravity {z=-980}
   }
   n7 = RigidBody(ComponentPose=@n6) {
    bClampLinearTranslationLimitToRefPose=true
    bOverrideWorldGravity=true
    overridePhysicsAsset=&/F_MED_InnovatorSand_Physics
    simulationSpace=BaseBoneSpace
    alphaBoolBlend {blendOption=Linear}
    alphaScaleBiasClamp {bInterpResult=true}
    baseBoneRef {boneName=pelvis}
    componentAppliedLinearAccClamp=<10 10 10>
    overrideWorldGravity {z=-980}
   }
   n8 = Anim Dynamics - Bone: dyn_zipper_front_leg_joint(ComponentPose=@n7, Alpha=1, ExternalForce="0,0,0") {
    angularDampingOverride=0.9999
    angularSpringConstant=0.5
    bGravityOverrideInSimSpace=true
    bOverrideAngularDamping=true
    bOverrideLinearDamping=true
    linearDampingOverride=0.9999
    simulationSpace=RootRelative
    alphaBoolBlend {blendOption=Linear}
    boundBone {boneName=dyn_zipper_front_leg_joint}
    componentAppliedLinearAccClamp=<10000 10000 10000>
    componentLinearAccScale=<0.15 0.15 0.15>
    componentLinearVelScale=<0.15 0.15 0.15>
    gravityOverride=<1 1 1>
    physicsBodyDefinitions [{
      sphereCollisionRadius=0
      boundBone {boneName=dyn_zipper_front_leg_joint}
      boxExtents=<6 6 2>
      constraintSetup {angularLimitsMax=<0 0 75>, angularLimitsMin=<0 -90 -80>, angularTarget=<0 0 0>}
      localJointOffset=<2.5 0 0>
     }]
    relativeSpaceBone {boneName=root}
   }
   n9 = Anim Dynamics - Bone: dyn_zipper_side_leg_joint(ComponentPose=@n8, Alpha=1, ExternalForce="0,0,0") {
    angularDampingOverride=0.9999
    angularSpringConstant=0.5
    bGravityOverrideInSimSpace=true
    bOverrideAngularDamping=true
    bOverrideLinearDamping=true
    linearDampingOverride=0.9999
    simulationSpace=RootRelative
    alphaBoolBlend {blendOption=Linear}
    boundBone {boneName=dyn_zipper_side_leg_joint}
    componentAppliedLinearAccClamp=<10000 10000 10000>
    componentLinearAccScale=<0.15 0.15 0.15>
    componentLinearVelScale=<0.15 0.15 0.15>
    gravityOverride=<1 1 1>
    physicsBodyDefinitions [{
      sphereCollisionRadius=0
      boundBone {boneName=dyn_zipper_side_leg_joint}
      boxExtents=<6 6 2>
      constraintSetup {angularLimitsMax=<0 0 75>, angularLimitsMin=<0 -90 -80>, angularTarget=<0 0 0>}
      localJointOffset=<2.5 0 0>
     }]
    relativeSpaceBone {boneName=root}
   }
   n10 = Anim Dynamics - Bone: dyn_zipper_back_leg_joint(ComponentPose=@n9, Alpha=1, ExternalForce="0,0,0") {
    angularDampingOverride=0.9999
    angularSpringConstant=0.5
    bGravityOverrideInSimSpace=true
    bOverrideAngularDamping=true
    bOverrideLinearDamping=true
    linearDampingOverride=0.9999
    simulationSpace=RootRelative
    alphaBoolBlend {blendOption=Linear}
    boundBone {boneName=dyn_zipper_back_leg_joint}
    componentAppliedLinearAccClamp=<10000 10000 10000>
    componentLinearAccScale=<0.15 0.15 0.15>
    componentLinearVelScale=<0.15 0.15 0.15>
    gravityOverride=<1 1 1>
    physicsBodyDefinitions [{
      sphereCollisionRadius=0
      boundBone {boneName=dyn_zipper_back_leg_joint}
      boxExtents=<6 6 2>
      constraintSetup {angularLimitsMax=<0 0 75>, angularLimitsMin=<0 -90 -80>, angularTarget=<0 0 0>}
      localJointOffset=<2.5 0 0>
     }]
    relativeSpaceBone {boneName=root}
   }
   n11 = Control Rig!(
    Source=Component To Local(ComponentPose=@n10)
    Alpha
    back_volume:double=Get back_volume() -> (back_volume:double)
    jumping:double=Get jumping_alpha() -> (jumping_alpha:double)
    crouching_Alpha:double=Get crouching_Alpha() -> (crouching_Alpha:double)
   ) -> (Pose) {
    alphaBoolBlend {blendOption=Linear}
    controlRigAssetReference {blueprintRigClass=&/F_MED_InnovatorSand_CtrlRig.F_MED_InnovatorSand_CtrlRig_C}
   }
   Output Pose(Result=@n11) {layerGroup=None}
  ]}
 EventGraph {nodes [
   Event BlueprintUpdateAnimation()
   n1 = Get bIsMoving2D()
   n2 = "Absolute (Float)"(A=BreakRotator(InRot=Get PawnDeltaRotation() -> (PawnDeltaRotation:Rotator)).Yaw)
   n3 = NOT Boolean(A=AND Boolean!(A="float > float"(A=@n2, B=2), B=Get bIsCrouching(), C:bool=NOT Boolean(A=@n1)) -> (ReturnValue))
   n4 = NOT Boolean(A=Get bIsMoving2D())
   n5 = Get bIsCrouching()
   > Set back_volume(back_volume:double="To Float (Boolean)"(InBool=AND Boolean(A=@n5, B=@n4, C=@n3))) -> (Output_Get:double)
   > Set jumping_alpha(jumping_alpha:double="To Float (Boolean)"(InBool=Get bIsJumping())) -> (Output_Get:double)
   > Set crouching_Alpha(crouching_Alpha:double="To Float (Boolean)"(InBool=Get bIsCrouching())) -> (Output_Get:double)
   > Set LOD_2_3(LOD_2_3:bool="integer >= integer"(A=Get CurrentLOD(), B=1)) -> (Output_Get:bool)
   Set static_crouching_rotation(static_crouching_rotation:bool=false) -> (Output_Get:bool)
  ]}
]
variables [
 back_volume {type=number}
 crouching_Alpha {type=number}
 jumping_alpha {type=number}
 LOD_2_3 {type=boolean}
 static_crouching_rotation {type=boolean}
]
